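# Physics asset for the collar of the Geisha female medium player cosmetic body, setting up its dynamic collar bones for physics.
# Seven box bodies cover the back and side collar bones, alongside capsules on the head and upper spine.
# Each collar body is constrained to spine_05 with limited swing and twist, generally looser at the front.
# Two extra angularly locked constraints chain the back, left-back and left-front "_a" pieces, with no right-side equivalent.
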PhysicsAsset /BRCosmetics/Characters/Player/Female/Medium/Bodies/F_MED_Geisha/Meshes/F_MED_Geisha_Collar_Physics
bodies {
 dyn_collar_bk [{shapeType=Box, extentX=3.1411, extentY=11.9286, extentZ=8.54, center=<-0.2015 0.2437 2.418>, rotation=<rot 0 -90.2797 -88.6853>}]
 dyn_collar_l_bk_a [{shapeType=Box, extentX=2.6982, extentY=6.8815, extentZ=11.5676, center=<-1.3098 -0.4556 1.8008>, rotation=<rot 0.6492 -82.6004 3.0913>}]
 dyn_collar_l_fr [{shapeType=Box, extentX=3.4224, extentY=8.4989, extentZ=13.6011, center=<1.2667 0.5841 4.4181>, rotation=<rot 18.3073 -79.4706 2.7379>}]
 dyn_collar_l_fr_a [{shapeType=Box, extentX=3.0294, extentY=7.3689, extentZ=12.8754, center=<-1.18 -0.9228 2.2597>, rotation=<rot 0.6336 -85.3182 0.0269>}]
 dyn_collar_r_bk_a [{shapeType=Box, extentX=2.1058, extentY=7.7758, extentZ=11.4806, center=<0.6335 0.3449 -2.1853>, rotation=<rot 0.077 -80.0063 -0.4862>}]
 dyn_collar_r_fr [{shapeType=Box, extentX=4.1092, extentY=10.394, extentZ=13.2639, center=<-1.0527 -2.6294 -5.367>, rotation=<rot 3.1439 -73.1761 -3.8036>}]
 dyn_collar_r_fr_a [{shapeType=Box, extentX=2.9288, extentY=6.7713, extentZ=12.7575, center=<1.2613 -0.4605 -2.9428>, rotation=<rot -12.9772 -86.4369 -0.1857>}]
 head [{shapeType=Capsule, length=6.5, radius=8.25, center=<3.8079 1.4274 0>, rotation=<rot 90 -5.0511 -5.0511>}]
 spine_05 [{shapeType=Capsule, length=1.5, radius=9}]
}
constraints [
 {bone1Name=dyn_collar_bk, bone2Name=spine_05, swing1LimitDegrees=5, swing2LimitDegrees=5, twistLimitDegrees=5}
 {bone1Name=dyn_collar_l_bk_a, bone2Name=spine_05, swing1LimitDegrees=5, swing2LimitDegrees=5, twistLimitDegrees=5}
 {bone1Name=dyn_collar_l_fr, bone2Name=spine_05, swing1LimitDegrees=30, swing2LimitDegrees=10, twistLimitDegrees=30}
 {bone1Name=dyn_collar_l_fr_a, bone2Name=spine_05, swing1LimitDegrees=5, swing2LimitDegrees=5, twistLimitDegrees=20}
 {bone1Name=dyn_collar_r_bk_a, bone2Name=spine_05, swing1LimitDegrees=5, swing2LimitDegrees=5, twistLimitDegrees=5}
 {bone1Name=dyn_collar_r_fr, bone2Name=spine_05, swing1LimitDegrees=15, swing2LimitDegrees=5, twistLimitDegrees=30}
 {bone1Name=dyn_collar_r_fr_a, bone2Name=spine_05, swing1LimitDegrees=10, swing2LimitDegrees=10, twistLimitDegrees=25}
 {
  bone1Name=dyn_collar_bk
  bone2Name=dyn_collar_l_bk_a
  swing1LimitDegrees=45
  swing1Motion=Locked
  swing2LimitDegrees=45
  swing2Motion=Locked
  twistLimitDegrees=45
  twistMotion=Locked
 }
 {
  bone1Name=dyn_collar_l_bk_a
  bone2Name=dyn_collar_l_fr_a
  swing1LimitDegrees=45
  swing1Motion=Locked
  swing2LimitDegrees=45
  swing2Motion=Locked
  twistLimitDegrees=45
  twistMotion=Locked
 }
]
