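# Physics asset for the female medium Geisha player body, set up for secondary motion of the coat's dynamic bones.
# Capsules approximate the pelvis, spine and thighs, while boxes cover two coat bones per side and a small left calf twist body.
# Coat bones have angular limits relative to spine_02, and each side's b bone is also rotationally locked to its a bone.
# Thigh-to-pelvis and calf-to-thigh constraints leave rotation free.
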PhysicsAsset /BRCosmetics/Characters/Player/Female/Medium/Bodies/F_MED_Geisha/Meshes/F_MED_Geisha_Physics
bodies {
 calf_twist_02_l [{shapeType=Box, extentX=1, extentY=1, extentZ=1}]
 dyn_coat_l_a [{shapeType=Box, extentX=3.5, extentY=9.5, extentZ=10, center=<0.009 0.7928 -4.1545>, rotation=<rot 5 10.0015 10.0006>}]
 dyn_coat_l_b [{shapeType=Box, extentX=3.5, extentY=9, extentZ=8, center=<-0.4363 -0.4245 -3.597>, rotation=<rot -9.9107 0.3285 -4.177>}]
 dyn_coat_r_a [{shapeType=Box, extentX=3, extentY=10.5, extentZ=9.5, center=<-0.3864 -0.4785 4.3075>, rotation=<rot -0.8673 14.927 9.9642>}]
 dyn_coat_r_b [{shapeType=Box, extentX=2.5, extentY=11.5, extentZ=9, center=<0 -0.0001 3.6553>, rotation=<rot -10 -10.0018 0>}]
 pelvis [{shapeType=Capsule, length=8.6042, radius=11.1339, center=<-4.2833 -2.0458 0.186>, rotation=<rot 0.5453 -4.7653 -172.5921>}]
 spine_01 [{shapeType=Capsule, length=6.75, radius=9.5, center=<-1.584 -0.6169 0>, rotation=<rot 0 -90 0>}]
 spine_02 [{shapeType=Capsule, length=2.5961, radius=9.079, center=<0.7861 0.078 0.0883>, rotation=<rot 0.9743 1.4864 -174.4651>}]
 spine_05 [{shapeType=Capsule, length=1.25, radius=11.4179, center=<-8.1956 1.2626 0.346>, rotation=<rot -0.415 -174.8105 -4.7498>}]
 thigh_l [{shapeType=Capsule, length=23.1158, radius=8.3485, center=<-23.4093 0.5175 0.3519>, rotation=<rot 0.7655 -95.1471 -90.6539>}]
 thigh_r [{shapeType=Capsule, length=24.5099, radius=8.1004, center=<22.4373 -0.3881 -0.3916>, rotation=<rot 0.3274 -95.0519 -90.6503>}]
}
constraints [
 {
  bone1Name=thigh_l
  bone2Name=pelvis
  swing1LimitDegrees=45
  swing1Motion=Free
  swing2LimitDegrees=45
  swing2Motion=Free
  twistLimitDegrees=45
  twistMotion=Free
 }
 {
  bone1Name=thigh_r
  bone2Name=pelvis
  swing1LimitDegrees=45
  swing1Motion=Free
  swing2LimitDegrees=45
  swing2Motion=Free
  twistLimitDegrees=45
  twistMotion=Free
 }
 {
  bone1Name=calf_twist_02_l
  bone2Name=thigh_l
  swing1LimitDegrees=45
  swing1Motion=Free
  swing2LimitDegrees=45
  swing2Motion=Free
  twistLimitDegrees=45
  twistMotion=Free
 }
 {bone1Name=dyn_coat_l_a, bone2Name=spine_02, swing1LimitDegrees=15, swing2LimitDegrees=45, twistLimitDegrees=15}
 {bone1Name=dyn_coat_l_b, bone2Name=spine_02, swing1LimitDegrees=25, swing2LimitDegrees=30, twistLimitDegrees=25}
 {bone1Name=dyn_coat_r_a, bone2Name=spine_02, swing1LimitDegrees=15, swing2LimitDegrees=45, twistLimitDegrees=15}
 {bone1Name=dyn_coat_r_b, bone2Name=spine_02, swing1LimitDegrees=25, swing2LimitDegrees=30, twistLimitDegrees=25}
 {
  bone1Name=dyn_coat_l_b
  bone2Name=dyn_coat_l_a
  swing1LimitDegrees=45
  swing1Motion=Locked
  swing2LimitDegrees=45
  swing2Motion=Locked
  twistLimitDegrees=45
  twistMotion=Locked
 }
 {
  bone1Name=dyn_coat_r_b
  bone2Name=dyn_coat_r_a
  swing1LimitDegrees=45
  swing1Motion=Locked
  swing2LimitDegrees=45
  swing2Motion=Locked
  twistLimitDegrees=45
  twistMotion=Locked
 }
]
