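# Animation Blueprint for the Geisha female body cosmetic, adding secondary motion to collar and shoestrings over the pose copied from the attached parent mesh.
# EventGraph derives correction alphas from head and upper-arm rotation relative to spine_05; thigh_rot_l is computed but unused.
# Additive bone offsets adjust the collar around two RigidBody passes, the collar-specific one weakening as arms rise.
# Four AnimDynamics spring chains then simulate the shoestrings.
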AnimBlueprint /BRCosmetics/Characters/Player/Female/Medium/Bodies/F_MED_Geisha/Meshes/F_MED_Geisha_AnimBP
target_skeleton=&/F_MED_Geisha_Skeleton
graphs [
 AnimGraph {nodes [
   n1 = "Transform (Modify) Bone - Bone: dyn_collar_r_fr_base"(Alpha=1, Translation="0,0,0", Rotation="0,0,0", Scale="1,1,1") {
    rotationMode=BMM_Additive
    rotationSpace=BCS_BoneSpace
    scaleMode=BMM_Additive
    scaleSpace=BCS_BoneSpace
    translationMode=BMM_Additive
    translationSpace=BCS_BoneSpace
    alphaBoolBlend {blendOption=Linear}
    alphaScaleBiasClamp {bMapRange=true, inRange {max=110, min=140}}
    boneToModify {boneName=dyn_collar_r_fr_base}
   }
   n2 = "Transform (Modify) Bone - Bone: dyn_collar_r_fr_a_base"(ComponentPose=@n1, Alpha=1, Translation="0,0,0", Rotation="0,0,-50", Scale="1,1,1") {
    rotationMode=BMM_Additive
    rotationSpace=BCS_BoneSpace
    scaleMode=BMM_Additive
    scaleSpace=BCS_BoneSpace
    translationMode=BMM_Additive
    translationSpace=BCS_BoneSpace
    alphaBoolBlend {blendOption=Linear}
    alphaScaleBiasClamp {bMapRange=true, inRange {max=110, min=140}}
    boneToModify {boneName=dyn_collar_r_fr_a_base}
    rotation {roll=-50}
   }
   "Transform (Modify) Bone - Bone: dyn_collar_r_bk_a_base"(ComponentPose=@n2, Alpha=1, Translation="0,0,0", Rotation="0,0,-50", Scale="1,1,1") {
    rotationMode=BMM_Additive
    rotationSpace=BCS_BoneSpace
    scaleMode=BMM_Additive
    scaleSpace=BCS_BoneSpace
    translationMode=BMM_Additive
    translationSpace=BCS_BoneSpace
    alphaBoolBlend {blendOption=Linear}
    alphaScaleBiasClamp {bMapRange=true, inRange {max=110, min=140}}
    boneToModify {boneName=dyn_collar_r_bk_a_base}
    rotation {roll=-50}
   }
   n3 = Get headRotBackClamp() -> (headRotBackClamp:double)
   n4 = "Transform (Modify) Bone - Bone: dyn_collar_bk_base"(
    ComponentPose=Local To Component(LocalPose=Copy Pose From Mesh() {bUseAttachedParent=true})
    Alpha=@n3
    Translation="0,0,0"
    Rotation="0,0,-50"
    Scale="1,1,1"
   ) {
    rotationMode=BMM_Additive
    rotationSpace=BCS_BoneSpace
    scaleMode=BMM_Additive
    scaleSpace=BCS_BoneSpace
    translationMode=BMM_Additive
    translationSpace=BCS_BoneSpace
    alphaBoolBlend {blendOption=Linear}
    boneToModify {boneName=dyn_collar_bk_base}
    rotation {roll=-50}
   }
   n5 = "Transform (Modify) Bone - Bone: dyn_collar_l_bk_a_base"(ComponentPose=@n4, Alpha=@n3, Translation="0,0,0", Rotation="0,0,-25", Scale="1,1,1") {
    rotationMode=BMM_Additive
    rotationSpace=BCS_BoneSpace
    scaleMode=BMM_Additive
    scaleSpace=BCS_BoneSpace
    translationMode=BMM_Additive
    translationSpace=BCS_BoneSpace
    alphaBoolBlend {blendOption=Linear}
    boneToModify {boneName=dyn_collar_l_bk_a_base}
    rotation {roll=-25}
   }
   n6 = "Transform (Modify) Bone - Bone: dyn_collar_r_bk_a_base"(ComponentPose=@n5, Alpha=@n3, Translation="0,0,0", Rotation="0,0,-25", Scale="1,1,1") {
    rotationMode=BMM_Additive
    rotationSpace=BCS_BoneSpace
    scaleMode=BMM_Additive
    scaleSpace=BCS_BoneSpace
    translationMode=BMM_Additive
    translationSpace=BCS_BoneSpace
    alphaBoolBlend {blendOption=Linear}
    boneToModify {boneName=dyn_collar_r_bk_a_base}
    rotation {roll=-25}
   }
   n7 = RigidBody(ComponentPose=@n6) {
    bClampLinearTranslationLimitToRefPose=true
    bOverrideWorldGravity=true
    overridePhysicsAsset=&/F_MED_Geisha_Physics
    simulationSpace=BaseBoneSpace
    alphaBoolBlend {blendOption=Linear}
    baseBoneRef {boneName=root}
    componentAppliedLinearAccClamp=<1500 1500 1500>
    componentLinearAccScale=<1 1 1>
    componentLinearVelScale=<1 1 1>
    overrideWorldGravity {z=-750}
   }
   n8 = RigidBody(ComponentPose=@n7, Alpha=Get collarDynamicBlend() -> (collarDynamicBlend:double)) {
    bClampLinearTranslationLimitToRefPose=true
    bOverrideWorldGravity=true
    overridePhysicsAsset=&/F_MED_Geisha_Collar_Physics
    simulationSpace=BaseBoneSpace
    alphaBoolBlend {blendOption=Linear}
    alphaScaleBiasClamp {bMapRange=true, outRange {max=0.2, min=1}}
    baseBoneRef {boneName=root}
    componentAppliedLinearAccClamp=<2000 2000 2000>
    componentLinearAccScale=<0.7 0.7 0.7>
    componentLinearVelScale=<0.7 0.7 0.7>
    overrideWorldGravity {z=-980}
   }
   n9 = Get l_armupClamp() -> (l_armupClamp:double)
   n10 = "Transform (Modify) Bone - Bone: dyn_collar_l_fr_base"(ComponentPose=@n8, Alpha=@n9, Translation="0,0,0", Rotation="0,0,17", Scale="1,1,1") {
    rotationMode=BMM_Additive
    rotationSpace=BCS_BoneSpace
    scaleMode=BMM_Additive
    scaleSpace=BCS_BoneSpace
    translationMode=BMM_Additive
    translationSpace=BCS_BoneSpace
    alphaBoolBlend {blendOption=Linear}
    boneToModify {boneName=dyn_collar_l_fr_base}
    rotation {roll=17}
   }
   n11 = "Transform (Modify) Bone - Bone: dyn_collar_l_fr_a_base"(ComponentPose=@n10, Alpha=@n9, Translation="0,0,0", Rotation="0,0,22", Scale="1,1,1") {
    rotationMode=BMM_Additive
    rotationSpace=BCS_BoneSpace
    scaleMode=BMM_Additive
    scaleSpace=BCS_BoneSpace
    translationMode=BMM_Additive
    translationSpace=BCS_BoneSpace
    alphaBoolBlend {blendOption=Linear}
    boneToModify {boneName=dyn_collar_l_fr_a_base}
    rotation {roll=22}
   }
   n12 = Get r_armupClamp() -> (r_armupClamp:double)
   n13 = "Transform (Modify) Bone - Bone: dyn_collar_r_fr_base"(ComponentPose=@n11, Alpha=@n12, Translation="0,-3.5,0", Rotation="0,0,45", Scale="1,1,1") {
    rotationMode=BMM_Additive
    rotationSpace=BCS_BoneSpace
    scaleMode=BMM_Additive
    scaleSpace=BCS_BoneSpace
    translationMode=BMM_Additive
    translationSpace=BCS_BoneSpace
    alphaBoolBlend {blendOption=Linear}
    boneToModify {boneName=dyn_collar_r_fr_base}
    rotation {roll=45}
    translation {y=-3.5}
   }
   n14 = "Transform (Modify) Bone - Bone: dyn_collar_r_fr_a_base"(ComponentPose=@n13, Alpha=@n12, Translation="0,0,0", Rotation="0,0,-10", Scale="1,1,1") {
    rotationMode=BMM_Additive
    rotationSpace=BCS_BoneSpace
    scaleMode=BMM_Additive
    scaleSpace=BCS_BoneSpace
    translationMode=BMM_Additive
    translationSpace=BCS_BoneSpace
    alphaBoolBlend {blendOption=Linear}
    boneToModify {boneName=dyn_collar_r_fr_a_base}
    rotation {roll=-10}
   }
   n15 = "Anim Dynamics - Chain: dyn_shoestring_l_a_1 -> dyn_shoestring_l_a_3"(ComponentPose=@n14, Alpha=1, ExternalForce="0,0,0") {
    angularDampingOverride=0.99
    angularSpringConstant=300
    bAngularSpring=true
    bChain=true
    bGravityOverrideInSimSpace=true
    bOverrideAngularDamping=true
    gravityScale=0.35
    simulationSpace=RootRelative
    alphaBoolBlend {blendOption=Linear}
    boundBone {boneName=dyn_shoestring_l_a_1}
    chainEnd {boneName=dyn_shoestring_l_a_3}
    componentAppliedLinearAccClamp=<100 100 100>
    componentLinearAccScale=<0.5 0.5 0.5>
    componentLinearVelScale=<0.5 0.5 0.5>
    physicsBodyDefinitions [{
      sphereCollisionRadius=0
      boundBone {boneName=dyn_shoestring_l_a_1}
      boxExtents=<5 10 10>
      constraintSetup {angularTargetAxis=AxisY, angularLimitsMax=<0 7 0>, angularLimitsMin=<0 -7 -50>, angularTarget=<0 0 0>}
      localJointOffset=<2 0 0>
     }]
   }
   n16 = "Anim Dynamics - Chain: dyn_shoestring_l_b_1 -> dyn_shoestring_l_b_3"(ComponentPose=@n15, Alpha=1, ExternalForce="0,0,0") {
    angularDampingOverride=0.99
    angularSpringConstant=300
    bAngularSpring=true
    bChain=true
    bGravityOverrideInSimSpace=true
    bOverrideAngularDamping=true
    gravityScale=0.35
    simulationSpace=RootRelative
    alphaBoolBlend {blendOption=Linear}
    boundBone {boneName=dyn_shoestring_l_b_1}
    chainEnd {boneName=dyn_shoestring_l_b_3}
    componentAppliedLinearAccClamp=<100 100 100>
    componentLinearAccScale=<0.5 0.5 0.5>
    componentLinearVelScale=<0.5 0.5 0.5>
    physicsBodyDefinitions [{
      sphereCollisionRadius=0
      boundBone {boneName=dyn_shoestring_l_b_1}
      boxExtents=<5 10 10>
      constraintSetup {angularTargetAxis=AxisY, angularLimitsMax=<0 5 0>, angularLimitsMin=<0 -5 -50>, angularTarget=<0 0 0>}
      localJointOffset=<1 0 0>
     }]
   }
   n17 = "Anim Dynamics - Chain: dyn_shoestring_r_a_1 -> dyn_shoestring_r_a_3"(ComponentPose=@n16, Alpha=1, ExternalForce="0,0,0") {
    angularDampingOverride=0.99
    angularSpringConstant=300
    bAngularSpring=true
    bChain=true
    bGravityOverrideInSimSpace=true
    bOverrideAngularDamping=true
    gravityScale=0.35
    simulationSpace=RootRelative
    alphaBoolBlend {blendOption=Linear}
    boundBone {boneName=dyn_shoestring_r_a_1}
    chainEnd {boneName=dyn_shoestring_r_a_3}
    componentAppliedLinearAccClamp=<100 100 100>
    componentLinearAccScale=<0.5 0.5 0.5>
    componentLinearVelScale=<0.5 0.5 0.5>
    physicsBodyDefinitions [{
      sphereCollisionRadius=0
      boundBone {boneName=dyn_shoestring_r_a_1}
      boxExtents=<5 10 10>
      constraintSetup {angularTargetAxis=AxisY, angularLimitsMax=<0 5 0>, angularLimitsMin=<0 -5 -50>, angularTarget=<0 0 0>}
      localJointOffset=<-1 0 0>
     }]
   }
   n18 = "Anim Dynamics - Chain: dyn_shoestring_r_b_1 -> dyn_shoestring_r_b_3"(ComponentPose=@n17, Alpha=1, ExternalForce="0,0,0") {
    angularDampingOverride=0.99
    angularSpringConstant=300
    bAngularSpring=true
    bChain=true
    bGravityOverrideInSimSpace=true
    bOverrideAngularDamping=true
    gravityScale=0.35
    simulationSpace=RootRelative
    alphaBoolBlend {blendOption=Linear}
    boundBone {boneName=dyn_shoestring_r_b_1}
    chainEnd {boneName=dyn_shoestring_r_b_3}
    componentAppliedLinearAccClamp=<100 100 100>
    componentLinearAccScale=<0.5 0.5 0.5>
    componentLinearVelScale=<0.5 0.5 0.5>
    physicsBodyDefinitions [{
      sphereCollisionRadius=0
      boundBone {boneName=dyn_shoestring_r_b_1}
      boxExtents=<5 10 10>
      constraintSetup {angularTargetAxis=AxisY, angularLimitsMax=<0 7 0>, angularLimitsMin=<0 -7 -50>, angularTarget=<0 0 0>}
      localJointOffset=<-2 0 0>
     }]
   }
   Output Pose(Result=Component To Local(ComponentPose=@n18)) {layerGroup=None}
  ]}
 EventGraph {nodes [
   Event BlueprintUpdateAnimation()
   > n1 = Sequence!(execute) -> (then_0, then_1, then_2:Exec, then_3:Exec, then_4:Exec)
   n2 = GetOwnerSkeletalMeshComponent()
   n3 = GetDeltaTransformFromRefPose(self=@n2, BoneName=upperarm_l, BaseName=spine_05)
   Set l_armUp(execute=@n1.then_1, l_armUp:double=BreakRotator(InRot=BreakTransform(InTransform=@n3).Rotation).Yaw) -> (Output_Get:double)
   > Set l_armupClamp(l_armupClamp:double=MapRangeClamped(Value=Get l_armUp() -> (l_armUp:double), InRangeA=80, InRangeB=110)) -> (Output_Get:double)
   n4 = GetDeltaTransformFromRefPose(self=@n2, BoneName=head, BaseName=spine_05)
   Set headFwdBk(execute=@n1.then_0, headFwdBk:double=BreakRotator(InRot=BreakTransform(InTransform=@n4).Rotation).Yaw) -> (Output_Get:double)
   > Set headRotBackClamp(headRotBackClamp:double=MapRangeClamped(Value=Get headFwdBk() -> (headFwdBk:double), InRangeA=-50, InRangeB=-20, OutRangeA=1, OutRangeB=0)) -> (Output_Get:double)
   n5 = GetDeltaTransformFromRefPose(self=@n2, BoneName=upperarm_r, BaseName=spine_05)
   Set r_armUp(execute=@n1.then_2, r_armUp:double=BreakRotator(InRot=BreakTransform(InTransform=@n5).Rotation).Yaw) -> (Output_Get:double)
   > Set r_armupClamp(r_armupClamp:double=MapRangeClamped(Value=Get r_armUp() -> (r_armUp:double), InRangeA=50, InRangeB=110)) -> (Output_Get:double)
   n6 = BreakTransform(InTransform=GetDeltaTransformFromRefPose(self=@n2, BoneName=thigh_l, BaseName=pelvis))
   Set thigh_rot_l(execute=@n1.then_3, thigh_rot_l=MapRangeClamped(Value=BreakRotator(InRot=@n6.Rotation).Yaw, InRangeA=-20, InRangeB=-100))
   n7 = Get l_armupClamp()
   n8 = Get r_armupClamp()
   Set collarDynamicBlend(
    execute=@n1.then_4
    collarDynamicBlend:double=float / float(A=float + float(A=@n7, B=@n8) {is=title K2Node_CommutativeAssociativeBinaryOperator}, B=2) {is=title K2Node_CallFunction}
   ) -> (Output_Get:double)
   TryGetPawnOwner()
  ]}
]
variables [
 collarDynamicBlend {type=number}
 headFwdBk {type=number}
 headRotBackClamp {type=number}
 l_armUp {type=number}
 l_armupClamp {type=number}
 r_armUp {type=number}
 r_armupClamp {type=number}
 thigh_rot_l {type=number}
]
