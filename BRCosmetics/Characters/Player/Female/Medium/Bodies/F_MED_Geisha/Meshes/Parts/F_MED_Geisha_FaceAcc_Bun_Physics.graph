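# Physics asset for the Geisha female character's bun face accessory, simulating the back hair pieces attached to the head.
# Capsule bodies cover the head and three back hair bones: center, left and right.
# Each hair body is constrained directly to the head with a tight first swing limit and wider second swing and twist limits, with the side pieces allowed more range than the center.
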PhysicsAsset /BRCosmetics/Characters/Player/Female/Medium/Bodies/F_MED_Geisha/Meshes/Parts/F_MED_Geisha_FaceAcc_Bun_Physics
bodies {
 dyn_hair_bk [{shapeType=Capsule, length=3.9067, radius=2.2126, center=<0.0001 -0.7905 0.8591>, rotation=<rot 0 0 -35.6852>}]
 dyn_hair_l_bk [{shapeType=Capsule, length=2.7464, radius=1.9178, center=<0.7783 -0.7219 0.814>, rotation=<rot -18.157 -89.9999 -18.5408>}]
 dyn_hair_r_bk [{shapeType=Capsule, length=2.2087, radius=2.6683, center=<-0.7214 1.1637 -0.8613>, rotation=<rot -37.8014 -90 -16.7327>}]
 head [{shapeType=Capsule, length=5, radius=5.4317, center=<1.8452 1.4827 -0.1346>, rotation=<rot -1.3837 80.2595 -92.0308>}]
}
constraints [
 {bone1Name=dyn_hair_l_bk, bone2Name=head, swing1LimitDegrees=5, swing2LimitDegrees=60, twistLimitDegrees=60}
 {bone1Name=dyn_hair_bk, bone2Name=head, swing1LimitDegrees=5, swing2LimitDegrees=32.5714, twistLimitDegrees=39.4286}
 {bone1Name=dyn_hair_r_bk, bone2Name=head, swing1LimitDegrees=5, swing2LimitDegrees=60, twistLimitDegrees=60}
]
