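# Physics asset for the female medium Geisha skin's face accessory mesh, defining two dangling hair chains, left and right.
# Each chain runs three capsules from a small box base constrained to the head with free swing and twist.
# Links have a very tight first swing limit and moderate second swing and twist, so they bend mostly in one plane.
# Head, neck and upper-spine capsules and slab-like boxes form unconstrained body shapes around the chains.
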PhysicsAsset /BRCosmetics/Characters/Player/Female/Medium/Bodies/F_MED_Geisha/Meshes/Parts/F_MED_Geisha_FaceAcc_Physics
bodies {
 dyn_hair_l_1 [{shapeType=Capsule, length=4.75, radius=1, center=<-0.3486 0.459 -3.1992>, rotation=<rot -5 0 -5.0001>}]
 dyn_hair_l_2 [{shapeType=Capsule, length=4.25, radius=0.75, center=<-0.2292 0.2786 -3.1267>, rotation=<rot -5 0 -5.0011>}]
 dyn_hair_l_3 [{shapeType=Capsule, length=4, radius=1, center=<-0.1388 1.075 -2.3052>, rotation=<rot -4.5305 2.1174 -25.0814>}]
 dyn_hair_l_base [{shapeType=Box, extentX=1.01, extentY=1.01, extentZ=1.01, rotation=<rot 0 -90 0>}]
 dyn_hair_r_1 [{shapeType=Capsule, length=4.75, radius=1, center=<0.4859 -0.6151 3.3006>, rotation=<rot -4.9238 0.8704 -5.039>}]
 dyn_hair_r_2 [{shapeType=Capsule, length=3.75, radius=0.75, center=<0.2084 -0.2634 3.0171>, rotation=<rot -4.9238 0.8704 -5.0411>}]
 dyn_hair_r_3 [{shapeType=Capsule, length=5, radius=1, center=<0.0693 -0.9774 2.0964>, rotation=<rot -4.5305 2.1174 -25.0807>}]
 dyn_hair_r_base [{shapeType=Box, extentX=1.01, extentY=1.01, extentZ=1.01, rotation=<rot 0 -90 0>}]
 head [{shapeType=Capsule, length=5.5, radius=4.9317, center=<1.8452 1.4827 -0.1346>, rotation=<rot -1.3837 80.2595 -92.0308>}]
 neck_01 [
  {shapeType=Capsule, length=7.25, radius=5.25, center=<-2.4062 0.1764 0.2107>, rotation=<rot 84.9989 -0.0001 20.0004>}
  {shapeType=Box, extentX=27.5, extentY=4, extentZ=33, center=<-0.1955 -0.7729 0>, rotation=<rot 0 -18.5855 0>}
 ]
 spine_05 [
  {shapeType=Capsule, length=12.5, radius=10.5, center=<0.5125 -3.3742 0>, rotation=<rot 0 5.0007 0>}
  {shapeType=Box, extentX=33, extentY=23, extentZ=3, center=<4.5286 6.6458 -8.599>, rotation=<rot -12.1643 14.6175 12.7078>}
  {shapeType=Box, extentX=35, extentY=13, extentZ=1.5, center=<6.958 -6.3802 -0.0702>, rotation=<rot -1.8481 3.4239 85.3252>}
  {shapeType=Box, extentX=31.5, extentY=22.5, extentZ=4, center=<6.6066 7.4303 8.6496>, rotation=<rot 9.2055 10.1266 171.4798>}
 ]
}
constraints [
 {bone1Name=dyn_hair_l_1, bone2Name=dyn_hair_l_base, swing1LimitDegrees=1, swing2LimitDegrees=20, twistLimitDegrees=35}
 {bone1Name=dyn_hair_l_2, bone2Name=dyn_hair_l_1, swing1LimitDegrees=1, swing2LimitDegrees=25, twistLimitDegrees=25}
 {bone1Name=dyn_hair_l_3, bone2Name=dyn_hair_l_2, swing1LimitDegrees=1, swing2LimitDegrees=25, twistLimitDegrees=25}
 {bone1Name=dyn_hair_r_1, bone2Name=dyn_hair_r_base, swing1LimitDegrees=1, swing2LimitDegrees=20, twistLimitDegrees=35}
 {bone1Name=dyn_hair_r_2, bone2Name=dyn_hair_r_1, swing1LimitDegrees=1, swing2LimitDegrees=25, twistLimitDegrees=25}
 {bone1Name=dyn_hair_r_3, bone2Name=dyn_hair_r_2, swing1LimitDegrees=1, swing2LimitDegrees=25, twistLimitDegrees=25}
 {
  bone1Name=dyn_hair_r_base
  bone2Name=head
  swing1LimitDegrees=45
  swing1Motion=Free
  swing2LimitDegrees=45
  swing2Motion=Free
  twistLimitDegrees=45
  twistMotion=Free
 }
 {
  bone1Name=dyn_hair_l_base
  bone2Name=head
  swing1LimitDegrees=45
  swing1Motion=Free
  swing2LimitDegrees=45
  swing2Motion=Free
  twistLimitDegrees=45
  twistMotion=Free
 }
]
