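# Post-process-style animation Blueprint for the Geisha cosmetic's face accessory part, adding secondary motion to its hair and bun.
# It copies the parent mesh pose, aims both hair base bones at spine_05, then runs two RigidBody simulations with separate physics assets.
# SwingVelocity, computed from spine_05's position relative to hand_r, drives the first simulation's alpha through an inverted range.
# Other AnimDynamics and Look At chains sit unconnected to the output.
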AnimBlueprint /BRCosmetics/Characters/Player/Female/Medium/Bodies/F_MED_Geisha/Meshes/Parts/F_MED_Geisha_FaceAcc_AnimBP
target_skeleton=&/F_MED_Geisha_FaceAcc_Skeleton
graphs [
 AnimGraph {nodes [
   n1 = Anim Dynamics - Bone: dyn_hair_bk(Alpha=1, ExternalForce="0,0,0") {
    angularDampingOverride=1
    angularSpringConstant=300
    bAngularSpring=true
    bGravityOverrideInSimSpace=true
    bOverrideAngularDamping=true
    bOverrideLinearDamping=true
    gravityScale=0.35
    simulationSpace=BoneRelative
    alphaBoolBlend {blendOption=Linear}
    boundBone {boneName=dyn_hair_bk}
    componentAppliedLinearAccClamp=<1000 1000 1000>
    componentLinearAccScale=<0.5 0.5 0.5>
    componentLinearVelScale=<0.5 0.5 0.5>
    physicsBodyDefinitions [{
      sphereCollisionRadius=0
      boundBone {boneName=dyn_hair_bk}
      boxExtents=<20 20 5>
      constraintSetup {twistAxis=AxisZ, angularLimitsMax=<10 0 0>, angularLimitsMin=<-10 0 0>, angularTarget=<0 0 0>}
      localJointOffset=<0 0 -8>
     }]
    relativeSpaceBone {boneName=head}
   }
   n2 = Anim Dynamics - Bone: dyn_hair_l_bk(ComponentPose=@n1, Alpha=1, ExternalForce="0,0,0") {
    angularDampingOverride=1
    angularSpringConstant=300
    bAngularSpring=true
    bGravityOverrideInSimSpace=true
    bOverrideAngularDamping=true
    bOverrideLinearDamping=true
    gravityScale=0.35
    alphaBoolBlend {blendOption=Linear}
    boundBone {boneName=dyn_hair_l_bk}
    componentAppliedLinearAccClamp=<3 3 3>
    componentLinearAccScale=<0.5 0.5 0.5>
    componentLinearVelScale=<0.5 0.5 0.5>
    physicsBodyDefinitions [{
      sphereCollisionRadius=0
      boundBone {boneName=dyn_hair_l_bk}
      boxExtents=<10 10 5>
      constraintSetup {angularTargetAxis=AxisY, angularLimitsMax=<0 2 1>, angularLimitsMin=<0 -2 -1>, angularTarget=<0 0 0>}
      localJointOffset=<0 0 -5>
     }]
   }
   Anim Dynamics - Bone: dyn_hair_r_bk(ComponentPose=@n2, Alpha=1, ExternalForce="0,0,0") {
    angularDampingOverride=1
    angularSpringConstant=300
    bAngularSpring=true
    bGravityOverrideInSimSpace=true
    bOverrideAngularDamping=true
    bOverrideLinearDamping=true
    gravityScale=0.35
    alphaBoolBlend {blendOption=Linear}
    boundBone {boneName=dyn_hair_r_bk}
    componentAppliedLinearAccClamp=<3 3 3>
    componentLinearAccScale=<0.5 0.5 0.5>
    componentLinearVelScale=<0.5 0.5 0.5>
    physicsBodyDefinitions [{
      sphereCollisionRadius=0
      boundBone {boneName=dyn_hair_r_bk}
      boxExtents=<10 10 5>
      constraintSetup {angularTargetAxis=AxisY, angularLimitsMax=<0 2 1>, angularLimitsMin=<0 -2 -1>, angularTarget=<0 0 0>}
      localJointOffset=<0 0 5>
     }]
   }
   n3 = Look At - Bone: dyn_hair_l_2(Alpha=1) {
    bUseLookUpAxis=true
    alphaBoolBlend {blendOption=Linear}
    boneToModify {boneName=dyn_hair_l_2}
    lookAtLocation {x=0}
    lookAtTarget {boneReference {boneName=spine_05}}
    lookAt_Axis {axis {y=0, z=-1}}
   }
   n4 = Look At - Bone: dyn_hair_l_3(ComponentPose=@n3, Alpha=1) {
    bUseLookUpAxis=true
    alphaBoolBlend {blendOption=Linear}
    boneToModify {boneName=dyn_hair_l_3}
    lookAtLocation {x=0}
    lookAtTarget {boneReference {boneName=spine_05}}
    lookAt_Axis {axis {y=0, z=-1}}
   }
   "Anim Dynamics - Chain: dyn_hair_r_1 -> dyn_hair_r_3"(ComponentPose=@n4, Alpha=1, ExternalForce="0,0,0") {
    angularDampingOverride=0.999
    angularSpringConstant=300
    bAngularSpring=true
    bChain=true
    bGravityOverrideInSimSpace=true
    bOverrideAngularDamping=true
    bUseSphericalLimits=true
    gravityScale=0.35
    alphaBoolBlend {blendOption=Linear}
    boundBone {boneName=dyn_hair_r_1}
    chainEnd {boneName=dyn_hair_r_3}
    componentAppliedLinearAccClamp=<1000 1000 1000>
    componentLinearAccScale=<0.3 0.3 0.3>
    componentLinearVelScale=<0.3 0.3 0.3>
    physicsBodyDefinitions [{
      sphereCollisionRadius=0
      boundBone {boneName=dyn_hair_r_1}
      boxExtents=<10 10 5>
      constraintSetup {twistAxis=AxisZ, angularLimitsMax=<35 30 0>, angularLimitsMin=<-35 30 0>, angularTarget=<0 0 0>}
      localJointOffset=<0 0 -3>
     }]
    planarLimits [
     {drivingBone {boneName=spine_05}, planeTransform {location=<0 0 3>, rotation=<rot 15 0 169.9999>}}
     {drivingBone {boneName=spine_05}, planeTransform {location=<0 0 -5>, rotation=<rot -7 0 15.0002>}}
    ]
    sphericalLimits [
     {limitRadius=6, drivingBone {boneName=head}, sphereLocalOffset=<0 0 0>}
     {limitRadius=13, drivingBone {boneName=spine_05}, sphereLocalOffset=<-1 -4 0>}
    ]
   }
   n5 = Look At - Bone: dyn_hair_r_2(Alpha=1) {
    bUseLookUpAxis=true
    alphaBoolBlend {blendOption=Linear}
    boneToModify {boneName=dyn_hair_r_2}
    lookAtLocation {x=0}
    lookAtTarget {boneReference {boneName=spine_05}}
    lookAt_Axis {axis {y=0, z=1}}
   }
   Look At - Bone: dyn_hair_r_3(ComponentPose=@n5, Alpha=1) {
    bUseLookUpAxis=true
    alphaBoolBlend {blendOption=Linear}
    boneToModify {boneName=dyn_hair_r_3}
    lookAtLocation {x=0}
    lookAtTarget {boneReference {boneName=spine_05}}
    lookAt_Axis {axis {y=0, z=1}}
   }
   n6 = Look At - Bone: dyn_hair_r_base(ComponentPose=Local To Component(LocalPose=Copy Pose From Mesh() {bUseAttachedParent=true}), Alpha=1) {
    bUseLookUpAxis=true
    alphaBoolBlend {blendOption=Linear}
    boneToModify {boneName=dyn_hair_r_base}
    lookAtLocation {x=0}
    lookAtTarget {boneReference {boneName=spine_05}}
    lookAt_Axis {axis {y=0, z=1}}
   }
   n7 = Look At - Bone: dyn_hair_l_base(ComponentPose=@n6, Alpha=1) {
    bUseLookUpAxis=true
    alphaBoolBlend {blendOption=Linear}
    boneToModify {boneName=dyn_hair_l_base}
    lookAtLocation {x=0}
    lookAtTarget {boneReference {boneName=spine_05}}
    lookAt_Axis {axis {y=0, z=-1}}
   }
   n8 = RigidBody(ComponentPose=@n7, Alpha=Get SwingVelocity() -> (SwingVelocity:double)) {
    bClampLinearTranslationLimitToRefPose=true
    bOverrideWorldGravity=true
    overridePhysicsAsset=&/F_MED_Geisha_FaceAcc_Physics
    simulationSpace=BaseBoneSpace
    alphaBoolBlend {blendOption=Linear}
    alphaScaleBiasClamp {bMapRange=true, outRange {max=0.6, min=1}}
    baseBoneRef {boneName=spine_03}
    componentAppliedLinearAccClamp=<100 100 100>
    componentLinearAccScale=<0.1 0.1 0.1>
    componentLinearVelScale=<0.1 0.1 0.1>
    overrideWorldGravity {z=-980}
   }
   n9 = RigidBody(ComponentPose=@n8) {
    bClampLinearTranslationLimitToRefPose=true
    bOverrideWorldGravity=true
    overridePhysicsAsset=&/F_MED_Geisha_FaceAcc_Bun_Physics
    simulationSpace=BaseBoneSpace
    alphaBoolBlend {blendOption=Linear}
    baseBoneRef {boneName=root}
    componentAppliedLinearAccClamp=<900 1300 1300>
    componentLinearAccScale=<0.3 0.7 0.7>
    componentLinearVelScale=<0.3 0.7 0.7>
    overrideWorldGravity {z=-800}
   }
   Output Pose(Result=Component To Local(ComponentPose=@n9)) {layerGroup=None}
   Get BodyMesh() -> (BodyMesh:Skeletal Mesh Component Object Reference)
  ]}
 EventGraph {nodes [
   Event BlueprintUpdateAnimation()
   n1 = GetPartSkeletalMeshComponent()
   > Is Valid(InputObject=@n1)
   n2 = GetAttachParent!(self:Scene Component Object Reference=@n1) -> (ReturnValue:Scene Component Object Reference)
   > Is Valid(InputObject=@n2)
   n3 = GetSocketTransform(self=@n2, InSocketName=hand_r, TransformSpace=RTS_World)
   n4 = GetSocketTransform(self=@n2, InSocketName=spine_05, TransformSpace=RTS_World)
   > n5 = Calculate Velocity From Position History!(
    execute
    DeltaSeconds:float=0.001
    Position:Vector=BreakTransform(InTransform=MakeRelativeTransform!(A:"Transform (by ref)"=@n4, RelativeTo:"Transform (by ref)"=@n3) -> (ReturnValue:Transform))
    History:"Position History Structure (by ref)"=Make Position History!() -> (PositionHistory:Position History Structure) {is=title K2Node_MakeStruct}
    NumberOfSamples:int=30
    VelocityMin:float=0.f
    VelocityMax:float=2500
   ) -> (then, ReturnValue:float)
   > Set SwingVelocity(SwingVelocity:double=@n5) -> (Output_Get:double)
   n6 = TryGetPawnOwner()
   Is Valid(InputObject=@n6)
   > n7 = Cast To PlayerPawn_Athena(Object=@n6)
   > Do Once()
   > n8 = Set PawnOwner(PawnOwner:Player Pawn Athena Object Reference=@n7) -> (Output_Get:Player Pawn Athena Object Reference)
   > Set BodyMesh(BodyMesh=GetSkeletalMeshForPartType(self=@n8))
  ]}
]
variables [
 BodyMesh {type=SkeletalMeshComponent}
 PawnOwner {type=PlayerPawn_Athena_C}
 SwingVelocity {type=number}
]
